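# Translucent, unlit, two-sided material for the placed volume of the Creative Player Counter device.
# UV channels 1 and 2 are rescaled by object scale around a center offset, masked by their U range and summed to sample the shared DeviceZone barrier texture.
# Its green channel, as a bottom gradient, adds to an inverted double-sided Fresnel.
# Emissive blends BG and bottom glow colors by Fresnel and inverted normal Z, through MF_FX_EmissiveAdjustments.
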Material /Game/Creative/Devices/PlayerCounter/Materials/M_CP_PlayerCounter_Volume_Placed_01
graph {
 properties {
  bUsedWithMeshParticles=true
  blendMode=BLEND_Translucent
  pixelDepthOffsetMode=PDOM_Legacy
  shadingModel=MSM_Unlit
  twoSided=true
  usedShadingModels=Unlit
 }
 nodes [
  n1 = Constant() {r=1}
  n2 = ComponentMask(None=TextureCoordinate() {coordinateIndex=2}) {r=true}
  n3 = TextureCoordinate() {coordinateIndex=2}
  n4 = AppendVector(
   A=Divide(
    A=Constant() {r=1}
    B=MaterialFunctionCall() {materialFunction=&ObjectScale}.Scale Y
   )
   B=Divide(
    A=Constant() {desc="0.85 is default, not 1", r=2}
    B=MaterialFunctionCall() {materialFunction=&ObjectScale}.Scale Z
   )
  )
  n5 = Constant() {r=1}
  n6 = Subtract(A=Add(A=Divide(A=@n3, B=@n4), B=@n5), B=Divide(A=@n5, B=@n4) {desc=Scale the center offset by the scaling factor}) {desc=Re offset to maintain center}
  n7 = Constant()
  n8 = Constant() {r=1}
  n9 = ComponentMask(None=TextureCoordinate() {coordinateIndex=1}) {r=true}
  n10 = TextureCoordinate() {coordinateIndex=1}
  n11 = AppendVector(
   A=Divide(
    A=Constant() {r=1}
    B=MaterialFunctionCall() {materialFunction=&ObjectScale}.Scale X
   )
   B=Divide(
    A=Constant() {desc="0.85 is default, not 1", r=2}
    B=MaterialFunctionCall() {materialFunction=&ObjectScale}.Scale Z
   )
  )
  n12 = Constant() {r=1}
  n13 = Subtract(A=Add(A=Divide(A=@n10, B=@n11), B=@n12), B=Divide(A=@n12, B=@n11) {desc=Scale the center offset by the scaling factor}) {desc=Re offset to maintain center}
  n14 = Constant()
  n15 = TextureSample(UVs=Add(A=If(A=@n1, B=@n2, "A > B"=@n6, "A < B"=@n7) {desc=0}, B=If(A=@n8, B=@n9, "A > B"=@n13, "A < B"=@n14) {desc=0})) {
   samplerSource=SSM_Clamp_WorldGroupSettings
   texture=/Game/Creative/Devices/DeviceZone_Base/Materials/Textures/T_CP_DeviceZone_Bounds_01_BarrierPlacer_D
  }
  n16 = MaterialFunctionCall(
   Brightness=ScalarParameter() {defaultValue=0.25, parameterName=FresnelBrightness}
   Exponent=ScalarParameter() {defaultValue=0.5, parameterName=FresnelExponent}
  ) {materialFunction=&MF_Fresnel_FX}
  Add(
   A=@n16.Fresnel Inverted Double Sided
   B=Multiply(A=@n15.G, B=ScalarParameter() {defaultValue=0.1, parameterName=BottomGradientOpacity}) {constB=0.5}
  )
  n17 = VectorParameter() {group=BG, parameterName=BG Color, sortPriority=1, defaultValue=<rgba 0 0.0072 0.5625 1>}
  n18 = VertexNormalWS()
  n19 = Multiply(A=OneMinus(None=ComponentMask(None=@n18) {b=true}), B=@n16.Fresnel Inverted Double Sided)
  n20 = VectorParameter() {group=BG, parameterName=BG Bottom Glow Color, sortPriority=2, defaultValue=<rgba 0 0.005 0.0851 1>}
  MaterialFunctionCall(Emissive=Multiply(A=LinearInterpolate(A=@n17.RGB, B=@n20.RGB, Alpha=@n19))) {materialFunction=&MF_FX_EmissiveAdjustments}
 ]
}
